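# Unlit translucent Niagara mesh-particle material for the eye spline effects in the HeavyRoar emote.
# SplineThicken sets strip width from a Thickness parameter, while emissive blends from black to particle color along scaled V.
# An opacity-style mask multiplies particle alpha, a panning smoke trail, a clamped soft-smoke texture and a sphere-mask fade across V.
# Dynamic parameters offset the pan time per particle and slide the soft mask along V.
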Material /BRCosmetics/Animation/Game/MainPlayer/Emotes/HeavyRoar/FX/M_FX_HeavyRoar_Eyesplines
graph {
 properties {
  bUsedWithNiagaraMeshParticles=true
  blendMode=BLEND_Translucent
  pixelDepthOffsetMode=PDOM_Legacy
  shadingModel=MSM_Unlit
  usedShadingModels=Unlit
 }
 nodes [
  n1 = ScalarParameter() {parameterName=Thickness}
  MaterialFunctionCall(WidthBase=@n1, WidthTip=@n1) {materialFunction=&SplineThicken}
  n2 = ParticleColor()
  n3 = ComponentMask(None=TextureCoordinate() {uTiling=2}) {g=true}
  n4 = ScalarParameter() {parameterName=e}
  n5 = Multiply(A=@n3, B=@n4)
  MaterialFunctionCall(Emissive=LinearInterpolate(A=Constant3Vector(), B=@n2.RGB, Alpha=Saturate(None=@n5))) {materialFunction=&MF_FX_EmissiveAdjustments}
  n6 = Time()
  n7 = DynamicParameter() {paramNames [MaskOffset,Rand,Param3,Param4]}
  n8 = TextureSample(UVs=Panner(
    Coordinate=MaterialFunctionCall() {materialFunction=&Swizzle}.YX
    Time=Add(A=@n6, B=@n7.Rand)
   ) {bFractionalPart=true, speedX=0.2}) {samplerType=SAMPLERTYPE_Grayscale, texture=&T_Smoke_Trail_Soft}
  n9 = Constant2Vector() {g=1}
  n10 = DynamicParameter() {paramNames [MaskOffset,Rand,Param3,Param4]}
  n11 = TextureSample(UVs=Add(A=TextureCoordinate() {vTiling=0.5}, B=Multiply(A=@n9, B=@n10.MaskOffset))) {samplerSource=SSM_Clamp_WorldGroupSettings, texture=/Game/Effects/Fort_Effects/Textures/T_Smoke_Soft}
  n12 = TextureCoordinate()
  Multiply(
   A=Multiply(A=@n2.A, B=Multiply(A=@n8.G, B=@n11.G))
   B=SphereMask(A=ComponentMask(None=@n12) {g=true}, B=Constant() {r=0.5}) {attenuationRadius=0.5, hardnessPercent=80}
  )
  ParticleRandom()
  Power(Base=@n3, Exp=@n4)
 ]
}
