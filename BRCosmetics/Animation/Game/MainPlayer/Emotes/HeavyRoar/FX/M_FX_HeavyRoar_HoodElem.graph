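# Translucent, unlit, two-sided particle material for a hood element in the HeavyRoar emote effect, usable on Niagara mesh particles.
# Particle color feeds an emissive adjustment function.
# Particle alpha is multiplied by an inverted Fresnel and panning inky smoke, projected on a camera-facing virtual plane at the particle position and offset by dynamic Erode.
# Erode also scales a negated local-space vector built from UV channels 1 and 2.
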Material /BRCosmetics/Animation/Game/MainPlayer/Emotes/HeavyRoar/FX/M_FX_HeavyRoar_HoodElem
graph {
 properties {
  bUsedWithNiagaraMeshParticles=true
  blendMode=BLEND_Translucent
  pixelDepthOffsetMode=PDOM_Legacy
  shadingModel=MSM_Unlit
  twoSided=true
  usedShadingModels=Unlit
 }
 nodes [
  n1 = ParticleColor()
  MaterialFunctionCall(Emissive=@n1.RGB) {materialFunction=&MF_FX_EmissiveAdjustments}
  n2 = StaticBool() {value=true}
  n3 = ScalarParameter() {defaultValue=1, parameterName=FresnelExp}
  n4 = ParticlePositionWS()
  n5 = MaterialFunctionCall(
   Plane Normal Axis=Subtract(A=CameraPositionWS(), B=@n4)
   Plane Center=@n4
   UV Scale=ScalarParameter() {defaultValue=100, parameterName=SmokeScale}
  ) {materialFunction=&VirtualPlaneCoordinates}
  n6 = VectorParameter() {parameterName=SmokeScale, defaultValue=<rgba 1 1 0 1>}
  n7 = Multiply(A=ComponentMask(None=VertexInterpolator(VS=@n5.UVs).PS) {g=true, r=true}, B=ComponentMask(None=@n6.RGB) {g=true, r=true})
  n8 = AppendVector(A=@n6.B, B=@n6.B)
  n9 = Add(
   A=TextureSample(UVs=Panner(Coordinate=@n7, Speed=@n8) {bFractionalPart=true, speedY=0.1}) {texture=/Game/Effects/Fort_Effects/Textures/Tile/T_Inky_Smoke_Tile2}.G
   B=DynamicParameter() {paramNames [Erode,Param2,Param3,Param4]}.Erode
  )
  Multiply(
   A=@n1.A
   B=Multiply(
    A=MaterialFunctionCall(Invert Fresnel=@n2, Power=@n3) {materialFunction=&Fresnel_Function}.Result
    B=Saturate(None=@n9)
   )
  )
  n10 = AppendVector(A=TextureCoordinate() {coordinateIndex=1}, B=TextureCoordinate() {coordinateIndex=2})
  Multiply(
   A=Multiply(A=Transform(None=@n10) {transformSourceType=TRANSFORMSOURCE_Local}) {constB=-1}
   B=DynamicParameter() {paramNames [Erode,Param2,Param3,Param4]}.Erode
  )
  TextureCoordinate()
 ]
}
