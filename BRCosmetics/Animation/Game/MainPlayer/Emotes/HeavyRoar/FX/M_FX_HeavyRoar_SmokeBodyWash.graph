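# Translucent unlit Niagara material that washes scrolling smoke over the character for the HeavyRoar emote, tinted by particle color.
# Two panned smoke layers, the first distorting the second, feed erosion and a distorted soft-smoke opacity mask.
# bUseBodyProject enables depth-based bump-offset UVs, a depth sphere mask and radial erosion from world position behind translucency.
# A view-facing camera offset moves particles, and higher quality levels add depth fade.
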Material /BRCosmetics/Animation/Game/MainPlayer/Emotes/HeavyRoar/FX/M_FX_HeavyRoar_SmokeBodyWash
graph {
 properties {
  bUsedWithNiagaraMeshParticles=true
  bUsedWithNiagaraSprites=true
  blendMode=BLEND_Translucent
  pixelDepthOffsetMode=PDOM_Legacy
  shadingModel=MSM_Unlit
  usedShadingModels=Unlit
 }
 nodes [
  n1 = ParticleColor()
  MaterialFunctionCall(Emissive=Multiply(A=NamedRerouteUsage(), B=@n1.RGB)) {materialFunction=&MF_FX_EmissiveAdjustments}
  MaterialFunctionCall!(Distance Field 1, Distance Field 2) {materialFunction=/Engine/Functions/DistanceFields/Combiners/DistanceField_Union}
  n2 = Constant3Vector() {constant=<rgba 0 1 0 0>}
  n3 = CameraPositionWS()
  n4 = ParticlePositionWS()
  n5 = DotProduct(A=Normalize(VectorInput=Transform(None=@n2) {transformSourceType=TRANSFORMSOURCE_Local}), B=Normalize(VectorInput=Subtract(A=@n3, B=@n4)))
  MaterialFunctionCall(
   Offset Amount=LinearInterpolate(
    A=ScalarParameter() {parameterName=CameraOffset_Rear}
    B=ScalarParameter() {defaultValue=15, parameterName=CameraOffset}
    Alpha=Saturate(None=@n5)
   )
   World Position=ParticlePositionWS()
  ) {materialFunction=/Engine/Functions/Engine_MaterialFunctions02/WorldPositionOffset/CameraOffset}
  n6 = TextureCoordinate()
  n7 = PixelDepth()
  n8 = SceneDepth(UVs=ScreenPosition().ViewportUV)
  n9 = Distance(A=@n7.R, B=@n8.R)
  n10 = ScalarParameter() {defaultValue=1, parameterName=Distance}
  n11 = StaticSwitchParameter(True=BumpOffset(Coordinate=@n6, Height=Divide(A=@n9, B=@n10)), False=@n6) {defaultValue=true, parameterName=bUseBodyProject}
  n12 = VectorParameter() {parameterName=Layer1_Xfm, channelNames {a=Y Speed, b=X Speed, g=Y Scale, r=X Scale}, defaultValue=<rgba 0 0 0 1>}
  n13 = AppendVector(A=@n12.X Scale, B=@n12.X Scale)
  n14 = TextureSample(UVs=Panner(Coordinate=Multiply(A=@n11, B=@n13), Speed=AppendVector(A=@n12.X Speed, B=@n12.X Speed)) {bFractionalPart=true, speedY=0.2}) {texture=&T_Inky_Smoke_Tile}
  n15 = VectorParameter() {parameterName=Layer2_Xfm, channelNames {a=Y Speed, b=X Speed, g=Y Scale, r=X Scale}, defaultValue=<rgba 0 0 0 1>}
  n16 = Multiply(A=@n11, B=AppendVector(A=@n15.X Scale, B=@n15.X Scale))
  n17 = Multiply(A=@n14.R, B=ScalarParameter() {parameterName=Distortion})
  n18 = TextureSample(UVs=Panner(Coordinate=Add(A=@n16, B=@n17), Speed=AppendVector(A=@n15.X Speed, B=@n15.X Speed)) {bFractionalPart=true, speedY=0.2}) {texture=&T_Inky_Smoke_Tile}
  NamedRerouteDeclaration(None=Saturate(None=Add(A=@n14.R, B=@n18.R))) {name=SmokeLayers, nodeColor=<rgba 1 0.3984 0 1>}
  NamedRerouteDeclaration(None=StaticSwitchParameter(True=SphereMask(A=@n7.R, B=@n8.R) {attenuationRadius=250}, False=Constant() {r=1}) {defaultValue=true, parameterName=bUseBodyProject}) {name=CharacterMask, nodeColor=<rgba 0.4766 1 0 1>}
  n19 = DynamicParameter() {defaultValue=<rgba 0.25 1 1 1>, paramNames [SmokeAdditional, Param2, Param3, RadiusMultiplier]}
  n20 = AppendVector(A=@n19.SmokeAdditional, B=@n19.SmokeAdditional)
  n21 = Divide(
   A=Distance(
    A=MaterialFunctionCall() {materialFunction=&WorldPositionBehindTranslucency}.Result
    B=AppendVector(A=@n20, B=@n19.Param3)
   )
   B=Multiply(A=@n19.RadiusMultiplier, B=ScalarParameter() {parameterName=MaximumRadius})
  )
  n22 = TextureSample() {texture=&T_Inky_Smoke_Tile}
  n23 = ScalarParameter() {parameterName=MaxErode}
  n24 = DynamicParameter() {parameterIndex=1, defaultValue=<rgba 0.25 1 1 1>, paramNames [SmokeAdditional, Param2, Param3, RadiusMultiplier]}
  NamedRerouteDeclaration(None=Subtract(
    A=StaticSwitchParameter(True=@n21, False=@n22.R) {defaultValue=true, parameterName=bUseBodyProject}
    B=Multiply(A=@n23, B=@n24.SmokeAdditional)
   )) {name=ErosionTime, nodeColor=<rgba 0 0.7656 1 1>}
  n25 = NamedRerouteUsage()
  n26 = NamedRerouteUsage()
  n27 = SmoothStep(Min=@n26, Max=Add(A=@n26, B=ScalarParameter() {parameterName=sstep_w}), Value=NamedRerouteUsage())
  n28 = Multiply(A=NamedRerouteUsage()) {constB=2}
  n29 = Add(A=TextureCoordinate(), B=Multiply(A=Subtract(A=@n28), B=ScalarParameter() {parameterName=SmokeMaskDistort}))
  n30 = Multiply(
   A=Multiply(A=@n1.A, B=Multiply(A=@n25, B=@n27))
   B=Multiply(A=TextureSample(UVs=@n29) {texture=/Game/Effects/Fort_Effects/Textures/T_Smoke_Soft}.G) {constB=1.5}
  )
  n31 = DepthFade(Opacity=@n30) {fadeDistanceDefault=2}
  QualitySwitch(Default=@n30, Low=@n30, High=@n31, Medium=@n30, Epic=@n31)
  Reroute(None=@n30)
 ]
}
