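# Unlit translucent Niagara mesh-particle material for the smoke face wash in the HeavyRoar emote.
# Two panned inky-smoke layers are summed, then smoothstepped at a threshold driven by particle alpha for erosion; both use only X scale and speed for both axes.
# A dynamic-parameter SphereMask along V reveals the smoke, shaped by per-vertex inverted Fresnel and a V taper.
# A scrolling texture, weighted by V, builds a normal-direction vertex push.
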Material /BRCosmetics/Animation/Game/MainPlayer/Emotes/HeavyRoar/FX/M_FX_HeavyRoar_SmokeFaceWash
graph {
 properties {
  bUsedWithNiagaraMeshParticles=true
  blendMode=BLEND_Translucent
  pixelDepthOffsetMode=PDOM_Legacy
  shadingModel=MSM_Unlit
  usedShadingModels=Unlit
 }
 nodes [
  n1 = ParticleColor()
  MaterialFunctionCall(Emissive=Multiply(A=NamedRerouteUsage(), B=@n1.RGB)) {materialFunction=&MF_FX_EmissiveAdjustments}
  n2 = TextureCoordinate()
  n3 = VectorParameter() {parameterName=Layer1_Xfm, channelNames {a=Y Speed, b=X Speed, g=Y Scale, r=X Scale}, defaultValue=<rgba 0 0 0 1>}
  n4 = AppendVector(A=@n3.X Scale, B=@n3.X Scale)
  n5 = TextureSample(UVs=Panner(Coordinate=Multiply(A=@n2, B=@n4), Speed=AppendVector(A=@n3.X Speed, B=@n3.X Speed)) {bFractionalPart=true, speedY=0.2}) {texture=&T_Inky_Smoke_Tile}
  n6 = TextureCoordinate()
  n7 = VectorParameter() {parameterName=Layer2_Xfm, channelNames {a=Y Speed, b=X Speed, g=Y Scale, r=X Scale}, defaultValue=<rgba 0 0 0 1>}
  n8 = AppendVector(A=@n7.X Scale, B=@n7.X Scale)
  n9 = TextureSample(UVs=Panner(Coordinate=Multiply(A=@n6, B=@n8), Speed=AppendVector(A=@n7.X Speed, B=@n7.X Speed)) {bFractionalPart=true, speedY=0.2}) {texture=&T_Inky_Smoke_Tile}
  n10 = NamedRerouteDeclaration(None=Saturate(None=Add(A=@n5.R, B=@n9.R))) {name=SmokeLayers, nodeColor=<rgba 1 0.3984 0 1>}
  Multiply(B=@n10)
  Multiply(A=@n1.A)
  n11 = DynamicParameter() {defaultValue=<rgba 0.25 1 1 1>, paramNames [RevealPosition, SmokeAdditional, Param3, Param4]}
  n12 = SphereMask(A=ComponentMask(None=TextureCoordinate()) {g=true}, B=@n11.RevealPosition, Radius=ScalarParameter() {parameterName=RevealR}) {hardnessPercent=0}
  n13 = NamedRerouteUsage()
  n14 = SmoothStep(Min=@n13, Max=Add(A=@n13, B=ScalarParameter() {parameterName=SmokeSmoothstep_W}), Value=@n10)
  n15 = Subtract(A=CameraPositionWS(), B=WorldPosition().XYZ)
  n16 = VertexInterpolator(VS=MaterialFunctionCall(Camera Vector=Normalize(VectorInput=@n15), Invert Fresnel=StaticBool() {value=true}, Power=ScalarParameter() {parameterName=FresnelPower}) {materialFunction=&Fresnel_Function}.Result)
  n17 = OneMinus(None=ComponentMask(None=TextureCoordinate()) {g=true})
  n18 = ScalarParameter() {parameterName=taper_scale}
  Multiply(A=@n12, B=Multiply(A=Multiply(A=@n14, B=@n16.PS), B=Multiply(A=@n17, B=@n18)))
  n19 = TextureSample(UVs=Panner(Coordinate=TextureCoordinate()) {speedY=0.5}) {texture=&T_Smooth_tile}
  n20 = ScalarParameter() {parameterName=VertexPush}
  n21 = TextureCoordinate()
  Multiply(A=VertexNormalWS(), B=Multiply(A=Multiply(A=@n19.R, B=@n20), B=ComponentMask(None=@n21) {g=true}))
  NamedRerouteDeclaration(None=@n12) {name=SmokeRollingMask, nodeColor=<rgba 0 1 0.9063 1>}
  NamedRerouteDeclaration(None=LinearInterpolate(A=ScalarParameter() {parameterName=step_min}, B=ScalarParameter() {parameterName=step_max}, Alpha=@n1.A)) {name=SmokeSmoothstepPos, nodeColor=<rgba 0 0.7187 1 1>}
  Saturate(None=Add(A=@n14, B=@n11.SmokeAdditional))
  ScalarParameter() {parameterName=RevealP}
  ScalarParameter() {parameterName=Opacity}
 ]
}
