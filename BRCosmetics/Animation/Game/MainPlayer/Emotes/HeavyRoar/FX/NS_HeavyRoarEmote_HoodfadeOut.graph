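# Niagara system for the HeavyRoar player emote that, by its name, fades out a hood effect.
# Its single Hood emitter burst-spawns a short-lived mesh particle that follows a sampled, offset bone location and rotation on User.SkeletalMesh every update.
# Over its lifetime the mesh grows slightly while curves raise then lower both its alpha and a dynamic material parameter.
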NiagaraSystem /BRCosmetics/Animation/Game/MainPlayer/Emotes/HeavyRoar/FX/NS_HeavyRoarEmote_HoodfadeOut
dependencies {
 usedDataInterfaces [&NiagaraDataInterfaceCurve]
 usedDynamicInputs []
 usedModules [
  &SystemState
  &Timeline
  &self:SystemSpawnScript.SystemScriptSource.SystemScriptGraph.NiagaraNodeAssignment_3.SetVariables_#1
 ]
}
emitters {Hood {
  emitterUpdateScript [
   &EmitterState(Life Cycle Mode=::NewEnumerator1, Loop Behavior=::NewEnumerator1, Loop Duration=1, UseLoopDelay=false)
   &SpawnBurst_Instantaneous(Use Spawn Probability=false, UseLoopCountLimit=false)
  ]
  particleSpawnScript [&InitializeParticle(
    Color=<rgba 0 0 0 1>
    Lifetime=1.2
    Mesh Scale Mode=::NewEnumerator3
    Sprite Size Mode=::NewEnumerator3
    Uniform Sprite Size=50
    UsePositionOffset=false
    Write Material Random=true
   )]
  particleUpdateScript [
   &ParticleState()
   &ScaleMeshSize(Scale Factor=&VectorFromFloat(Value=&FloatFromCurve(FloatCurve=DataInterface() {propertyValues {Curve {keys [{interpMode=RCIM_Cubic, tangentMode=RCTM_User, time=0, value=1}, {interpMode=RCIM_Cubic, tangentMode=RCTM_User, time=1, value=1.125}]}}})))
   &ScaleColor(
    Scale Alpha=&FloatFromCurve(FloatCurve=DataInterface() {propertyValues {Curve {keys [<at 0 0 RCIM_Cubic>,<at 0.2 1 RCIM_Cubic>,<at 1 0 RCIM_Cubic>]}}}, Scale Curve=2)
    ScaleA=true
    ScaleRGB=false
   )
   &SolveForcesAndVelocity(Clamp Velocity=false, Limit Acceleration=false)
   &SkeletalMeshLocation(Bone Sampling Mode=::NewEnumerator4, MaskSpawnGroup=false, Sampled Position Offset=<-8 2 0>, Skeletal Mesh=$User.SkeletalMesh)
   /SetVariables(Particles.MeshOrientation=&MultiplyQuaternion(Quaternion A=$StackContext.SkeletalMeshLocation.SampledRotation, Quaternion B=<0 0 0 1>)) {isSetParameters=true}
   &CameraOffset()
   &DynamicMaterialParameters(
    Index 0 Param 1=&Lerp_Float(
     A=0
     B=1
     Alpha=&FloatFromCurve(FloatCurve=DataInterface() {propertyValues {Curve {keys [<at 0 0 RCIM_Cubic>,<at 0.3 1 RCIM_Cubic>,<at 1 0 RCIM_Cubic>]}}})
    )
    Module.Param0WriteEnabled=true
    Module.Param1WriteEnabled=true
    Module.Param2WriteEnabled=true
    Module.Param3WriteEnabled=true
   )
  ]
  rendererClasses [&NiagaraMeshRendererProperties]
  renderers [{rendererClass=&NiagaraMeshRendererProperties, rendererIndex=0}]
 }}
summary {
 systemName=NS_HeavyRoarEmote_HoodfadeOut
 emitters [{emitterName=Hood, rendererClasses [&NiagaraMeshRendererProperties]}]
 userVariables [
  User.bIsFrontEnd {type=NiagaraBool}
  User.ImpactColor {type=LinearColor, description="", defaultValue {struct=LinearColor, value=<rgba 1 1 1 1>}}
  User.IsWeakPoint {type=NiagaraBool}
  User.NormalizedNotifyProgress {type=NiagaraFloat, defaultValue=0, description=""}
  User.SkeletalMesh {
   type=NiagaraDataInterfaceSkeletalMesh
   defaultValue {
    struct=NiagaraExt_VariableValue_DataInterface
    value {dataInterface=&self:NiagaraDataInterfaceSkeletalMesh_0}
   }
  }
 ]
}
