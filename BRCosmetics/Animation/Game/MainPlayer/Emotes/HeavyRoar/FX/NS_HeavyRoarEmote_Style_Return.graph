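# Niagara effect for the Return style variant of the Heavy Roar emote, layering dark smoke and glowing eyes on the character.
# System-level skeletal sampling and HeadSource anchor face, body and splash smoke, with splash size switched by bIsFrontEnd.
# Eye glows, light and embers use spawn groups for mirrored bone offsets, while separate sprites cover the hands.
# Smoke washes feed world positions into dynamic material parameters; one body wash is disabled.
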NiagaraSystem /BRCosmetics/Animation/Game/MainPlayer/Emotes/HeavyRoar/FX/NS_HeavyRoarEmote_Style_Return
dependencies {
 usedDataInterfaces [
  &NiagaraDataInterfaceMeshRendererInfo
  &NiagaraDataInterfaceCurve
  &NiagaraDataInterfaceVector2DCurve
  &NiagaraDataInterfaceColorCurve
  &NiagaraDataInterfaceArrayFloat3
  &NiagaraDataInterfaceArrayFloat
  &NiagaraDataInterfaceCamera
 ]
 usedDynamicInputs [
  &RandomRangeVector
  &RotateVectorByQuaternion
  /Niagara/DynamicInputs/Math/TransformVectorByMatrix
  /Niagara/DynamicInputs/Math/MakeTransform
 ]
 usedModules [
  &self:SystemSpawnScript.SystemScriptSource.SystemScriptGraph.NiagaraNodeAssignment_3.SetVariables_#1
  &SystemState
  &self:SystemSpawnScript.SystemScriptSource.SystemScriptGraph.NiagaraNodeAssignment_0.SetVariables_#2
  &self:SystemSpawnScript.SystemScriptSource.SystemScriptGraph.NiagaraNodeAssignment_1.SetVariables_#3
  &SampleSkeletalMesh
  &self:SystemSpawnScript.SystemScriptSource.SystemScriptGraph.NiagaraNodeAssignment_2.SetVariables_#4
 ]
}
emitters {
 BodySmokeWash {
  bEnabled=false
  emitterUpdateScript [
   &EmitterState(Life Cycle Mode=::NewEnumerator1, Loop Behavior=::NewEnumerator1, Loop Duration=1, UseLoopDelay=false)
   &SpawnBurst_Instantaneous(Use Spawn Probability=false, UseLoopCountLimit=false)
  ]
  particleSpawnScript [&InitializeParticle(
    Color=<rgba 0 0 0 1.35>
    Lifetime=2
    Position Offset=<0 0 -9.3583>
    Sprite Size=<80 135>
    Sprite Size Mode=::NewEnumerator4
    UsePositionOffset=true
    Write Material Random=true
   )]
  particleUpdateScript [
   &ParticleState()
   &SolveForcesAndVelocity(Clamp Velocity=false, Limit Acceleration=false)
   /SetVariables(Particles.Position=$System.SampleSkeletalMesh.SampledPosition) {enabled=false, isSetParameters=true, module=SetVariables_#5}
   /SetVariables(Particles.MeshOrientation=$System.SampleSkeletalMesh.SampledRotation) {isSetParameters=true, module=SetVariables_#6}
   &CameraOffset(Camera Offset Amount=100)
   &DynamicMaterialParameters(
    Index 0 Param 1=&MakeFloatFromVector(Vector=Linked() {linkedVariable {name=System.SampleSkeletalMesh.SampledPosition, type=NiagaraPosition}}, Channel=::NewEnumerator0)
    Index 0 Param 2=&MakeFloatFromVector(Vector=Linked() {linkedVariable {name=System.SampleSkeletalMesh.SampledPosition, type=NiagaraPosition}}, Channel=::NewEnumerator1)
    Index 0 Param 3=&MakeFloatFromVector(Vector=Linked() {linkedVariable {name=System.SampleSkeletalMesh.SampledPosition, type=NiagaraPosition}}, Channel=::NewEnumerator2)
    Index 0 Param 4=&FloatFromCurve(
     FloatCurve=DataInterface() {propertyValues {Curve {keys [{arriveTangent=-0.75, interpMode=RCIM_Cubic, time=0, value=1}, {interpMode=RCIM_Cubic, tangentMode=RCTM_User, time=1, value=-0.75}]}}}
     Scale Curve=10
    )
    Index 1 Param 1=&FloatFromCurve(FloatCurve=DataInterface() {propertyValues {Curve {keys [<at 0 0.275 RCIM_Cubic>, {arriveTangent=-0.275, interpMode=RCIM_Cubic, leaveTangent=-0.275, time=0.2, value=1.25}, <at 1 0 RCIM_Cubic>]}}})
    Module.Param0WriteEnabled=true
    Module.Param0WriteEnabled1=true
    Module.Param1WriteEnabled=true
    Module.Param1WriteEnabled1=true
    Module.Param2WriteEnabled=true
    Module.Param2WriteEnabled1=true
    Module.Param3WriteEnabled=true
    Module.Param3WriteEnabled1=true
    Write Parameter Index 1=true
   )
   &SpriteFacingAndAlignment(Sprite Alignment=<0 0 1>, WriteAlignment=true, WriteFacing=true)
   &ScaleColor(
    Scale Alpha=&FloatFromCurve(FloatCurve=DataInterface() {propertyValues {Curve {keys [{arriveTangent=-1.1, interpMode=RCIM_Cubic, time=0, value=0}, <at 0.1 1.05>, <at 1 0 RCIM_Cubic>]}}})
    ScaleA=true
    ScaleRGB=true
   )
  ]
  rendererClasses [&NiagaraSpriteRendererProperties]
  renderers [{rendererClass=&NiagaraSpriteRendererProperties}]
 }
 BodySmokeWash001 {
  emitterUpdateScript [
   &EmitterState(Life Cycle Mode=::NewEnumerator1, Loop Behavior=::NewEnumerator1, Loop Duration=1, UseLoopDelay=false)
   &SpawnBurst_Instantaneous(Use Spawn Probability=false, UseLoopCountLimit=false)
  ]
  particleSpawnScript [&InitializeParticle(
    Color=$System.PaleBlack
    Lifetime=1.5
    Position Offset=<0 0 86.4798>
    Sprite Size=<60 80>
    Sprite Size Mode=::NewEnumerator4
    UsePositionOffset=true
    Write Material Random=true
   )]
  particleUpdateScript [
   &ParticleState()
   &ScaleSpriteSize(
    Scale Factor=&MakeVector2D(
     X=1
     Y=&FloatFromCurve(FloatCurve=DataInterface() {propertyValues {Curve {keys [<at 0 0.6 RCIM_Cubic>,<at 1 1.075 RCIM_Cubic>]}}})
    )
    Scale Sprite Size Mode=::NewEnumerator2
   )
   &SolveForcesAndVelocity(Clamp Velocity=false, Limit Acceleration=false)
   /SetVariables(Particles.Position=$System.SampleSkeletalMesh.SampledPosition) {enabled=false, isSetParameters=true, module=SetVariables_#5}
   /SetVariables(Particles.MeshOrientation=$System.SampleSkeletalMesh.SampledRotation) {isSetParameters=true, module=SetVariables_#6}
   /SetVariables(Particles.Position=&AddVectorToPosition(Position=$System.HeadSource, Vector=<0 0 5>)) {isSetParameters=true}
   &CameraOffset(Camera Offset Amount=15) {enabled=false}
   &DynamicMaterialParameters(
    Index 0 Param 1=&MakeFloatFromVector(Vector=Linked() {linkedVariable {name=System.HeadSource, type=NiagaraPosition}}, Channel=::NewEnumerator0)
    Index 0 Param 2=&MakeFloatFromVector(Vector=Linked() {linkedVariable {name=System.HeadSource, type=NiagaraPosition}}, Channel=::NewEnumerator1)
    Index 0 Param 3=&MakeFloatFromVector(Vector=Linked() {linkedVariable {name=System.HeadSource, type=NiagaraPosition}}, Channel=::NewEnumerator2)
    Index 0 Param 4=&FloatFromCurve(FloatCurve=DataInterface() {propertyValues {Curve {keys [
         <at 0 0 RCIM_Cubic>
         {arriveTangent=1.9, interpMode=RCIM_Cubic, leaveTangent=1.9, time=0.1, value=1}
         {arriveTangent=-1.1111, interpMode=RCIM_Cubic, leaveTangent=-1.1111, time=0.5, value=0.95}
         <at 1 0 RCIM_Cubic>
        ]}}})
    Index 1 Param 1=&FloatFromCurve(FloatCurve=DataInterface() {propertyValues {Curve {keys [{interpMode=RCIM_Cubic, tangentMode=RCTM_User, time=0, value=1}, {interpMode=RCIM_Cubic, tangentMode=RCTM_User, time=1, value=0}]}}})
    Module.Param0WriteEnabled=true
    Module.Param0WriteEnabled1=true
    Module.Param1WriteEnabled=true
    Module.Param1WriteEnabled1=true
    Module.Param2WriteEnabled=true
    Module.Param2WriteEnabled1=true
    Module.Param3WriteEnabled=true
    Module.Param3WriteEnabled1=true
    Write Parameter Index 1=true
   )
   &SpriteFacingAndAlignment(Sprite Alignment=<0 0 1>, WriteAlignment=true, WriteFacing=true)
   /SetVariables(Particles.FullSpriteSize:Vector2f=&Multiply_Vector2D_ByFloat(Float=0.7, Vector2D=$StackContext.SpriteSize)) {isSetParameters=true}
   &ScaleColor(
    Scale Alpha=&FloatFromCurve(
     FloatCurve=DataInterface() {propertyValues {Curve {keys [
         {arriveTangent=0.6875, interpMode=RCIM_Cubic, time=0, value=0}
         {arriveTangent=2.5, interpMode=RCIM_Cubic, leaveTangent=2.5, time=0.2, value=1}
         {arriveTangent=-1.6, interpMode=RCIM_Cubic, leaveTangent=-1.6, time=0.4, value=1}
         {arriveTangent=-1.6667, interpMode=RCIM_Cubic, leaveTangent=-1.6667, time=0.7, value=0.2}
         <at 1 0 RCIM_Cubic>
        ]}}}
     Scale Curve=0.9
    )
    ScaleA=true
    ScaleRGB=true
   )
  ]
  rendererClasses [&NiagaraSpriteRendererProperties]
  renderers [
   {rendererClass=&NiagaraSpriteRendererProperties}
   {rendererClass=&NiagaraSpriteRendererProperties, rendererIndex=1}
   {rendererClass=&NiagaraSpriteRendererProperties, rendererIndex=2}
  ]
 }
 EyeEmbers {
  emitterUpdateScript [
   &EmitterState(Life Cycle Mode=::NewEnumerator1, Loop Behavior=::NewEnumerator1, Loop Delay=0.05, Loop Duration=0.85, UseLoopDelay=true)
   &SpawnBurst_Instantaneous(Spawn Count=10, Use Spawn Probability=false, UseLoopCountLimit=false) {enabled=false}
   &SpawnRate(
    SpawnRate=&FloatFromCurve(
     FloatCurve=DataInterface() {propertyValues {Curve {keys [<at 0 1 RCIM_Cubic>, {arriveTangent=-1.6667, interpMode=RCIM_Cubic, leaveTangent=-1.6667, time=0.1, value=1.5}, <at 0.6 0 RCIM_Cubic>]}}}
     CurveIndex=$Emitter.NormalizedLoopAge
     Scale Curve=70
    )
    Use Spawn Probability=false
   )
  ]
  particleSpawnScript [
   &InitializeParticle(
    Color=<rgba 6 6 6 1>
    Lifetime=&RandomRangeFloat(Minimum=0.2, Maximum=1)
    Sprite Size=<1 1>
    Sprite Size Mode=::NewEnumerator4
    UsePositionOffset=false
    Write Material Random=true
   )
   &SkeletalMeshLocation(
    Bone Sampling Mode=::NewEnumerator4
    Interpolate Mesh Sampling=false
    MaskSpawnGroup=false
    Sampled Position Offset=&Multiply_Vector(
     A=<4 10 2.75>
     B=&SelectVectorFromArray(
      Array Sampling Mode=::NewEnumerator0
      Direct Array Index=&SpawnGroup()
      Vector Selection Array=DataInterface() {propertyValues {FloatData [<1 1 1>, <1 1 -1>]}}
      Random Seed=0
     )
    )
    Skeletal Mesh=$User.SkeletalMesh
   )
   &AddVelocity(
    Module.Use Falloff Distance=false
    Override Seed=false
    Velocity Mode=::NewEnumerator1
    Velocity Origin=$StackContext.Position
    Velocity Speed=&RandomRangeFloat(Minimum=5, Maximum=35)
   )
   &AddVelocity(Velocity=<-15 0 10>) {module=AddVelocity001}
  ]
  particleUpdateScript [
   &ParticleState()
   &InheritSourceMovement(
    Applied Position Falloff Distance=900
    Applied Position Scale=&VectorFromFloat(Value=&FloatFromCurve(FloatCurve=DataInterface() {propertyValues {Curve {keys [<at 0 1 RCIM_Cubic>]}}}, Scale Curve=10))
    ApplyForce=true
    ApplyForceFalloff=false
    ApplyPosition=true
    ApplyPositionFalloff=true
    ApplyPositionFalloffCurve=false
    Module.LimitForceSpeed=false
   )
   &Drag(Drag=&RandomRangeFloat(Minimum=2, Maximum=5), Use Linear Drag=true, Use Rotational Drag=true)
   &GravityForce(Gravity=&Multiply_VectorByFloat(Vector=<0 0 400>, Float=&RandomRangeFloat(Minimum=-0.2, Maximum=0.5)))
   &CurlNoiseForce(Noise Frequency=1, Noise Quality / Cost=::NewEnumerator0, Noise Strength=50, Pan Noise=true, Pan Noise Field=<0 0 20>)
   &SolveForcesAndVelocity(Clamp Velocity=false, Limit Acceleration=false)
   &ScaleColor(
    Linear Color Curve=DataInterface() {propertyValues {
      AlphaCurve {keys [<at 0 1>,<at 1 0>]}
      BlueCurve {keys [<at 0 0>,<at 0.1257 0>]}
      GreenCurve {keys [<at 0 1.2095>,<at 0.1257 0>]}
      RedCurve {keys [<at 0 2>,<at 0.1257 8>]}
     }}
    Scale Mode=::NewEnumerator2
   )
   &ScaleSpriteSize(
    Scale Factor=&MakeVector2D(
     X=1
     Y=&FloatFromCurve(FloatCurve=DataInterface() {propertyValues {Curve {keys [<at 0 0 RCIM_Cubic>,<at 0.1 1 RCIM_Cubic>,<at 1 0 RCIM_Cubic>]}}})
    )
    Scale Sprite Size Mode=::NewEnumerator2
   )
   &ScaleSpriteSizeBySpeed(Max Scale Factor=<1 10>, Velocity Threshold=200)
   &ScaleSpriteSize(
    Scale Sprite Size Mode=::NewEnumerator0
    Uniform Scale Factor=&MakeCustomFloatFromBool(
     Bool=&SetBoolByFloatComparison(
      A=&VectorLength(Vector=&SubtractVectorFromPosition(
        Position=&ReturnCameraPosition(Coordinate Space:ENiagaraCoordinateSpace=::World, Camera Query:NiagaraDataInterfaceCamera=DataInterface())
        Vector=Linked() {linkedVariable {name=Particles.Position, type=NiagaraPosition}}
       ))
      B=125
      Comparison Type=::NewEnumerator6
     )
     True Float=0
     False Float=1
    )
   ) {module=ScaleSpriteSize001}
  ]
  rendererClasses [&NiagaraSpriteRendererProperties]
  renderers [
   {rendererClass=&NiagaraSpriteRendererProperties}
   {rendererClass=&NiagaraSpriteRendererProperties, rendererIndex=1}
  ]
 }
 EyeGlows {
  emitterUpdateScript [
   &EmitterState(Life Cycle Mode=::NewEnumerator1, Loop Behavior=::NewEnumerator1, Loop Duration=1, UseLoopDelay=false)
   &SpawnBurst_Instantaneous(Use Spawn Probability=false, UseLoopCountLimit=false)
   &SpawnBurst_Instantaneous(Spawn Group=1, Use Spawn Probability=false, UseLoopCountLimit=false) {module=SpawnBurst_Instantaneous001}
  ]
  particleSpawnScript [&InitializeParticle(
    Color=<rgba 1 1 1 0.5>
    Lifetime=0.75
    Sprite Size=<6 3>
    Sprite Size Mode=::NewEnumerator4
    UsePositionOffset=false
    Write Material Random=true
   )]
  particleUpdateScript [
   &ParticleState()
   &ScaleSpriteSize(
    Uniform Curve Scale=2
    Uniform Curve Sprite Scale=DataInterface() {propertyValues {Curve {keys [<at 0 1 RCIM_Cubic>,<at 0.1 1.35 RCIM_Cubic>,<at 1 1 RCIM_Cubic>]}}}
   )
   &SkeletalMeshLocation(
    Bone Sampling Mode=::NewEnumerator4
    MaskSpawnGroup=false
    Sampled Position Offset=&Multiply_Vector(
     A=<4 10 2.75>
     B=&SelectVectorFromArray(
      Direct Array Index=&SpawnGroup()
      Vector Selection Array=DataInterface() {propertyValues {FloatData [<1 1 1>, <1 1 -1>]}}
     )
    )
    Skeletal Mesh=$User.SkeletalMesh
   )
   &SolveForcesAndVelocity(Clamp Velocity=false, Limit Acceleration=false)
   &CameraOffset(Camera Offset Amount=4)
   &ScaleColor(
    Linear Color Curve=DataInterface() {propertyValues {
      bOptimizeLUT=false
      AlphaCurve {keys [<at 0 0.3>,<at 0.9607 0.7>]}
      BlueCurve {keys [<at 0 0.464>,<at 0.9636 0>]}
      GreenCurve {keys [<at 0 2.0937>,<at 0.9636 0>]}
      RedCurve {keys [<at 0 5>,<at 0.9636 0>]}
     }}
    Scale Mode=::NewEnumerator2
   )
   &ScaleColor(
    Color Value To Scale=$StackContext.Color
    Scale Alpha=&FloatFromCurve(FloatCurve=DataInterface() {propertyValues {Curve {keys [<at 0 0 RCIM_Cubic>,<at 0.1 1 RCIM_Cubic>,<at 1 0 RCIM_Cubic>]}}})
    ScaleA=true
    ScaleRGB=true
   ) {module=ScaleColor001}
  ]
  rendererClasses [&NiagaraSpriteRendererProperties, &NiagaraLightRendererProperties]
  renderers [
   {rendererClass=&NiagaraSpriteRendererProperties}
   {rendererClass=&NiagaraSpriteRendererProperties, rendererIndex=1}
   {rendererClass=&NiagaraLightRendererProperties, rendererIndex=2}
  ]
 }
 EyeGlows001 {
  emitterUpdateScript [
   &EmitterState(Life Cycle Mode=::NewEnumerator1, Loop Behavior=::NewEnumerator1, Loop Delay=0.1, Loop Duration=1, UseLoopDelay=true)
   &SpawnBurst_Instantaneous(Use Spawn Probability=false, UseLoopCountLimit=false)
   &SpawnBurst_Instantaneous(Spawn Group=1, Use Spawn Probability=false, UseLoopCountLimit=false) {module=SpawnBurst_Instantaneous001}
  ]
  particleSpawnScript [&InitializeParticle(
    Color=<rgba 1.7 1.7 1.7 1>
    Lifetime=3.2
    Sprite Size=<6 3>
    Sprite Size Mode=::NewEnumerator4
    UsePositionOffset=false
    Write Material Random=true
   )]
  particleUpdateScript [
   &ParticleState()
   &ScaleSpriteSize(
    Uniform Curve Scale=2
    Uniform Curve Sprite Scale=DataInterface() {propertyValues {Curve {keys [<at 0 1 RCIM_Cubic>,<at 0.1 1.35 RCIM_Cubic>,<at 1 1 RCIM_Cubic>]}}}
   )
   &SkeletalMeshLocation(
    Bone Sampling Mode=::NewEnumerator4
    MaskSpawnGroup=false
    Sampled Position Offset=&Multiply_Vector(
     A=<4 10 2.75>
     B=&SelectVectorFromArray(
      Direct Array Index=&SpawnGroup()
      Vector Selection Array=DataInterface() {propertyValues {FloatData [<1 1 1>, <1 1 -1>]}}
     )
    )
    Skeletal Mesh=$User.SkeletalMesh
   )
   /SetVariables(Particles.MeshOrientation=&MultiplyQuaternion(
     Quaternion A=$StackContext.SkeletalMeshLocation.SampledRotation
     Quaternion B=&MakeQuatFromAxisAngle(
      Axis=<1 0 0>
      AngleInDegrees=&SelectFloatFromArray(Direct Array Index=&SpawnGroup(), Float Selection Array=DataInterface() {propertyValues {FloatData [60,-60]}})
     )
    )) {isSetParameters=true}
   &SolveForcesAndVelocity(Clamp Velocity=false, Limit Acceleration=false)
   &CameraOffset(Camera Offset Amount=4)
   &ScaleColor(
    Linear Color Curve=DataInterface() {propertyValues {
      bOptimizeLUT=false
      AlphaCurve {keys [<at 0 0.3>,<at 0.9607 0.7>]}
      BlueCurve {keys [<at 0 0.464>,<at 0.1857 0>]}
      GreenCurve {keys [<at 0 2.0937>,<at 0.1857 0>]}
      RedCurve {keys [<at 0 5>,<at 0.1857 0>]}
     }}
    Scale Mode=::NewEnumerator2
   )
   &ScaleColor(
    Color Value To Scale=$StackContext.Color
    Scale Alpha=&FloatFromCurve(FloatCurve=DataInterface() {propertyValues {Curve {keys [<at 0 0 RCIM_Cubic>,<at 0.1 1 RCIM_Cubic>,<at 1 0 RCIM_Cubic>]}}})
    ScaleA=true
    ScaleRGB=true
   ) {module=ScaleColor001}
   &DynamicMaterialParameters(
    Index 0 Param 1=&FloatFromCurve(
     FloatCurve=DataInterface() {propertyValues {Curve {keys [{interpMode=RCIM_Cubic, tangentMode=RCTM_User, time=0, value=0}, {interpMode=RCIM_Cubic, tangentMode=RCTM_User, time=1, value=1}]}}}
     Scale Curve=8
    )
    Index 0 Param 2=&RandomRangeFloat(Minimum=0, Maximum=1)
    Module.Param0WriteEnabled=true
    Module.Param1WriteEnabled=true
    Module.Param2WriteEnabled=true
    Module.Param3WriteEnabled=true
   )
  ]
  rendererClasses [&NiagaraMeshRendererProperties]
  renderers [{rendererClass=&NiagaraMeshRendererProperties, rendererIndex=0}]
 }
 FlameSplashSmoke {
  emitterUpdateScript [
   &EmitterState(Life Cycle Mode=::NewEnumerator1, Loop Behavior=::NewEnumerator1, Loop Duration=2, UseLoopDelay=false)
   &SpawnBurst_Instantaneous(Use Spawn Probability=false, UseLoopCountLimit=false)
  ]
  particleSpawnScript [
   &InitializeParticle(
    Color=<rgba 1 1 1 0.35>
    Lifetime=1.7
    Sprite Rotation Angle=270
    Sprite Rotation Mode=::NewEnumerator0
    Sprite Size=&Multiply_Vector2D_ByFloat(Float=0.35, Vector2D=<130 120>)
    Sprite Size Mode=::NewEnumerator4
    UsePositionOffset=false
    Write Material Random=true
   )
   &ShapeLocation(
    MaskSpawnGroup=false
    Offset=<0.7488 20 0>
    Offset Coordinate Space=::Local
    Offset Mode=::NewEnumerator0
    Override Seed=false
    Shape Origin=$System.HeadSource
    Sphere Radius=1
    UseOwnerScale=true
   )
  ]
  particleUpdateScript [
   &ParticleState()
   &ScaleSpriteSize(
    Uniform Curve Scale=&MakeCustomFloatFromBool(Bool=$User.bIsFrontEnd, True Float=0.55, False Float=1.2)
    Uniform Curve Sprite Scale=DataInterface() {propertyValues {Curve {keys [
        {arriveTangent=1.9622, interpMode=RCIM_Cubic, time=0, value=0.6}
        {arriveTangent=0.5896, interpMode=RCIM_Cubic, leaveTangent=0.5896, tangentMode=RCTM_User, time=0.3, value=0.8}
        {interpMode=RCIM_Cubic, leaveTangent=0.0224, time=1, value=1}
       ]}}}
   )
   &ScaleColor(
    Scale Alpha=&FloatFromCurve(FloatCurve=DataInterface() {propertyValues {Curve {keys [
         {interpMode=RCIM_Cubic, tangentMode=RCTM_User, time=0, value=0}
         {interpMode=RCIM_Cubic, tangentMode=RCTM_User, time=0.3, value=1}
         {tangentMode=RCTM_User, time=1, value=0}
        ]}}})
    Scale RGB=&MakeVectorFromLinearColorRGB(Color=&ColorFromCurve(ColorCurve=DataInterface() {propertyValues {
        AlphaCurve {keys [<at 0.4833 1 RCIM_Cubic>]}
        BlueCurve {keys [<at 0 0>,<at 0.0867 0.05>]}
        GreenCurve {keys [<at 0 0>,<at 0.0867 0.05>]}
        RedCurve {keys [<at 0 0>,<at 0.0867 0.05>]}
       }}))
    ScaleA=true
    ScaleRGB=true
   )
   &DynamicMaterialParameters(
    Index 0 Param 1=&FloatFromCurve(
     FloatCurve=DataInterface() {propertyValues {Curve {keys [
         {arriveTangent=0.2738, interpMode=RCIM_Cubic, leaveTangent=0.2738, tangentMode=RCTM_User, time=0.1, value=0}
         {interpMode=RCIM_Cubic, tangentMode=RCTM_User, time=1, value=1}
        ]}}}
     Scale Curve=10
    )
    Index 0 Param 2=&FloatFromCurve(FloatCurve=DataInterface() {propertyValues {Curve {keys [
         {arriveTangent=0.0396, interpMode=RCIM_Cubic, leaveTangent=0.0396, tangentMode=RCTM_User, time=0, value=1.8}
         {arriveTangent=1.4204, interpMode=RCIM_Cubic, leaveTangent=1.4204, tangentMode=RCTM_User, time=1, value=2.5}
        ]}}})
    Index 0 Param 3=&FloatFromCurve(
     FloatCurve=DataInterface() {propertyValues {Curve {keys [
         {interpMode=RCIM_Cubic, tangentMode=RCTM_User, time=0, value=0}
         {arriveTangent=1.7963, interpMode=RCIM_Cubic, leaveTangent=1.7963, tangentMode=RCTM_User, time=1, value=1}
        ]}}}
     Scale Curve=0.1
    )
    Index 0 Param 4=&FloatFromCurve(
     FloatCurve=DataInterface() {propertyValues {Curve {keys [
         {arriveTangent=-0.768, interpMode=RCIM_Cubic, time=0, value=1}
         {arriveTangent=-0.875, interpMode=RCIM_Cubic, leaveTangent=-0.875, time=0.6, value=0.8}
         {interpMode=RCIM_Cubic, leaveTangent=-0.714, time=1, value=0.125}
        ]}}}
     Scale Curve=1.25
    )
    Module.Param0WriteEnabled=true
    Module.Param1WriteEnabled=true
    Module.Param2WriteEnabled=true
    Module.Param3WriteEnabled=true
   )
   &GravityForce(Gravity=<0 0 0>)
   /SetVariables(Particles.Position=&AddVectorToPosition(Position=$System.HeadSource, Vector=&TransformVector(Vector=<0 0 3.5>, Destination Space=::Simulation))) {isSetParameters=true}
   &CameraOffset(Camera Offset Amount=15)
   &SolveForcesAndVelocity(Clamp Velocity=false, Limit Acceleration=false)
  ]
  rendererClasses [&NiagaraSpriteRendererProperties]
  renderers [{rendererClass=&NiagaraSpriteRendererProperties}]
 }
 HandsHide {
  emitterUpdateScript [
   &EmitterState(Life Cycle Mode=::NewEnumerator1, Loop Behavior=::NewEnumerator1, Loop Duration=1, UseLoopDelay=false)
   &SpawnBurst_Instantaneous(Spawn Group=1, Use Spawn Probability=false, UseLoopCountLimit=false)
   &SpawnBurst_Instantaneous(Spawn Group=2, Use Spawn Probability=false, UseLoopCountLimit=false) {module=SpawnBurst_Instantaneous001}
  ]
  particleSpawnScript [
   &InitializeParticle(
    Color=<rgba 0.04 0.04 0.04 1.0615>
    Lifetime=&RandomRangeFloat(Minimum=0.9, Maximum=1.2)
    Sprite Size=<30 30>
    Sprite Size Mode=::NewEnumerator4
    UsePositionOffset=true
    Write Material Random=true
   )
   /SetVariables() {enabled=false, isSetParameters=true}
  ]
  particleUpdateScript [
   &ParticleState()
   &SolveForcesAndVelocity(Clamp Velocity=false, Limit Acceleration=false)
   &SkeletalMeshLocation(
    Bone / Socket Index=&SpawnGroup(Max Spawn Group=2)
    Bone Sampling Mode=::NewEnumerator4
    MaskSpawnGroup=false
    Sampled Position Offset=&SelectVectorFromArray(
     Direct Array Index=&SpawnGroup(Max Spawn Group=2)
     Vector Selection Array=DataInterface() {propertyValues {FloatData [<0 0 0>, <-9 0 0>, <9 0 0>]}}
    )
    Skeletal Mesh=$User.SkeletalMesh
   )
   /SetVariables(Particles.Position=$System.SampleSkeletalMesh.SampledPosition) {enabled=false, isSetParameters=true, module=SetVariables_#5}
   /SetVariables(Particles.MeshOrientation=$System.SampleSkeletalMesh.SampledRotation) {enabled=false, isSetParameters=true, module=SetVariables_#6}
   &CameraOffset(Camera Offset Amount=10)
   &DynamicMaterialParameters(
    Index 0 Param 1=&MakeFloatFromVector(Vector=Linked() {linkedVariable {name=Particles.Position, type=NiagaraPosition}}, Channel=::NewEnumerator0)
    Index 0 Param 2=&MakeFloatFromVector(Vector=Linked() {linkedVariable {name=Particles.Position, type=NiagaraPosition}}, Channel=::NewEnumerator1)
    Index 0 Param 3=&MakeFloatFromVector(Vector=Linked() {linkedVariable {name=Particles.Position, type=NiagaraPosition}}, Channel=::NewEnumerator2)
    Index 0 Param 4=&FloatFromCurve(FloatCurve=DataInterface() {propertyValues {Curve {keys [
         <at 0 0 RCIM_Cubic>
         {arriveTangent=1.9, interpMode=RCIM_Cubic, leaveTangent=1.9, time=0.1, value=1}
         {arriveTangent=-1.1111, interpMode=RCIM_Cubic, leaveTangent=-1.1111, time=0.5, value=0.95}
         <at 1 0 RCIM_Cubic>
        ]}}})
    Index 1 Param 1=&FloatFromCurve(FloatCurve=DataInterface() {propertyValues {Curve {keys [{interpMode=RCIM_Cubic, tangentMode=RCTM_User, time=0, value=1}, {interpMode=RCIM_Cubic, tangentMode=RCTM_User, time=1, value=0}]}}})
    Module.Param0WriteEnabled=true
    Module.Param0WriteEnabled1=true
    Module.Param1WriteEnabled=true
    Module.Param1WriteEnabled1=true
    Module.Param2WriteEnabled=true
    Module.Param2WriteEnabled1=true
    Module.Param3WriteEnabled=true
    Module.Param3WriteEnabled1=true
    Write Parameter Index 1=true
   )
   &ScaleColor(
    Scale Alpha=&FloatFromCurve(
     FloatCurve=DataInterface() {propertyValues {Curve {keys [
         <at 0 0 RCIM_Cubic>
         {arriveTangent=0.6875, interpMode=RCIM_Cubic, leaveTangent=0.6875, time=0.1, value=1.025}
         {arriveTangent=-1.1389, interpMode=RCIM_Cubic, leaveTangent=-1.1389, time=0.4, value=0.275}
         <at 1 0 RCIM_Cubic>
        ]}}}
     Scale Curve=1.5
    )
    ScaleA=true
    ScaleRGB=true
   ) {enabled=false}
  ]
  rendererClasses [&NiagaraSpriteRendererProperties]
  renderers [
   {rendererClass=&NiagaraSpriteRendererProperties}
   {rendererClass=&NiagaraSpriteRendererProperties, rendererIndex=1}
  ]
 }
 SmokeFaceWash {
  emitterUpdateScript [
   &EmitterState(Life Cycle Mode=::NewEnumerator1, Loop Behavior=::NewEnumerator1, Loop Duration=1, UseLoopDelay=false)
   &SpawnBurst_Instantaneous(Use Spawn Probability=false, UseLoopCountLimit=false)
  ]
  particleSpawnScript [&InitializeParticle(Color=<rgba 0.02 0.02 0.02 0.8>, Lifetime=1.1, Sprite Size Mode=::NewEnumerator3, UsePositionOffset=false, Write Material Random=true)]
  particleUpdateScript [
   &ParticleState()
   &SolveForcesAndVelocity(Clamp Velocity=false, Limit Acceleration=false)
   /SetVariables(Particles.Position=$System.SampleSkeletalMesh.SampledPosition) {isSetParameters=true, module=SetVariables_#5}
   /SetVariables(Particles.MeshOrientation=$System.SampleSkeletalMesh.SampledRotation) {isSetParameters=true, module=SetVariables_#6}
   &CameraOffset(Camera Offset Amount=15)
   &DynamicMaterialParameters(
    Index 0 Param 1=&Lerp_Float(
     A=1
     B=0
     Alpha=&FloatFromCurve(FloatCurve=DataInterface() {propertyValues {Curve {keys [<at 0 1 RCIM_Cubic>, {arriveTangent=-1, interpMode=RCIM_Cubic, leaveTangent=-1, time=0.5, value=0.575}, <at 1 0 RCIM_Cubic>]}}})
    )
    Module.Param0WriteEnabled=true
    Module.Param1WriteEnabled=true
    Module.Param2WriteEnabled=true
    Module.Param3WriteEnabled=true
   )
   &ScaleColor(
    Scale Alpha=&FloatFromCurve(FloatCurve=DataInterface() {propertyValues {Curve {keys [
         <at 0 0 RCIM_Cubic>
         {arriveTangent=2, interpMode=RCIM_Cubic, leaveTangent=2, time=0.1, value=1.125}
         {arriveTangent=-1.25, interpMode=RCIM_Cubic, leaveTangent=-1.25, time=0.5, value=1}
         {tangentMode=RCTM_User, time=1, value=0}
        ]}}})
    ScaleA=true
    ScaleRGB=true
   )
  ]
  rendererClasses [&NiagaraMeshRendererProperties]
  renderers [{rendererClass=&NiagaraMeshRendererProperties, rendererIndex=0}]
 }
}
summary {
 systemName=NS_HeavyRoarEmote_Style_Return
 emitters [
  {emitterName=FlameSplashSmoke, rendererClasses [&NiagaraSpriteRendererProperties]}
  {emitterName=SmokeFaceWash, rendererClasses [&NiagaraMeshRendererProperties]}
  {bEnabled=false, emitterName=BodySmokeWash, rendererClasses [&NiagaraSpriteRendererProperties]}
  {emitterName=EyeGlows, rendererClasses [&NiagaraSpriteRendererProperties, &NiagaraLightRendererProperties]}
  {emitterName=BodySmokeWash001, rendererClasses [&NiagaraSpriteRendererProperties]}
  {emitterName=EyeGlows001, rendererClasses [&NiagaraMeshRendererProperties]}
  {emitterName=HandsHide, rendererClasses [&NiagaraSpriteRendererProperties]}
  {emitterName=EyeEmbers, rendererClasses [&NiagaraSpriteRendererProperties]}
 ]
 userVariables [
  User.bIsFrontEnd {type=NiagaraBool}
  User.ImpactColor {type=LinearColor, description="", defaultValue {struct=LinearColor, value=<rgba 1 1 1 1>}}
  User.IsWeakPoint {type=NiagaraBool}
  User.SkeletalMesh {
   type=NiagaraDataInterfaceSkeletalMesh
   defaultValue {
    struct=NiagaraExt_VariableValue_DataInterface
    value {dataInterface=&self:NiagaraDataInterfaceSkeletalMesh_0}
   }
  }
 ]
}
