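# Niagara effect for the style transform in the HeavyRoar emote, masking the character with dark smoke while the eyes glow and shed embers.
# Head and hand smoke sprites, a face-wash mesh and splash smoke follow sampled skeletal mesh or head positions, feeding material fades.
# System ShowEyeGlow scales eye glows and tints splash smoke, whose size depends on bIsFrontEnd.
# Embers use camera-distance culling; older smoke, burst and debug emitters are disabled.
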NiagaraSystem /BRCosmetics/Animation/Game/MainPlayer/Emotes/HeavyRoar/FX/NS_HeavyRoarEmote_Style_Transform
dependencies {
 usedDataInterfaces [
  &NiagaraDataInterfaceCurve
  &NiagaraDataInterfaceMeshRendererInfo
  &NiagaraDataInterfaceVector2DCurve
  &NiagaraDataInterfaceColorCurve
  &NiagaraDataInterfaceArrayFloat3
  &NiagaraDataInterfaceCamera
 ]
 usedDynamicInputs [
  &RandomRangeVector
  /Niagara/DynamicInputs/Math/TransformVectorByMatrix
  /Niagara/DynamicInputs/Math/MakeTransform
  /Game/Effects/Niagara/DynamicInputs/Math/FN_RemapDotProduct
  &NormalizeVector
  &Subtract_Vector
 ]
 usedModules [
  &self:SystemSpawnScript.SystemScriptSource.SystemScriptGraph.NiagaraNodeAssignment_4.SetVariables_#1
  &SystemState
  &self:SystemSpawnScript.SystemScriptSource.SystemScriptGraph.NiagaraNodeAssignment_1.SetVariables_#2
  &self:SystemSpawnScript.SystemScriptSource.SystemScriptGraph.NiagaraNodeAssignment_0.SetVariables_#3
  &SampleSkeletalMesh
  &self:SystemSpawnScript.SystemScriptSource.SystemScriptGraph.NiagaraNodeAssignment_2.SetVariables_#4
  &self:SystemSpawnScript.SystemScriptSource.SystemScriptGraph.NiagaraNodeAssignment_3.SetVariables_#5
 ]
}
emitters {
 BodySmokeWash {
  bEnabled=false
  emitterUpdateScript [
   &EmitterState(Life Cycle Mode=::NewEnumerator1, Loop Behavior=::NewEnumerator1, Loop Duration=1, UseLoopDelay=false)
   &SpawnBurst_Instantaneous(Use Spawn Probability=false, UseLoopCountLimit=false)
  ]
  particleSpawnScript [&InitializeParticle(
    Color=<rgba 0.01 0.01 0.01 4>
    Lifetime=2.2
    Position Offset=<0 0 86.4798>
    Sprite Size=<80 60>
    Sprite Size Mode=::NewEnumerator4
    UsePositionOffset=true
    Write Material Random=true
   )]
  particleUpdateScript [
   &ParticleState()
   &SolveForcesAndVelocity(Clamp Velocity=false, Limit Acceleration=false)
   /SetVariables(Particles.Position=$System.SampleSkeletalMesh.SampledPosition) {enabled=false, isSetParameters=true, module=SetVariables_#6}
   /SetVariables(Particles.MeshOrientation=$System.SampleSkeletalMesh.SampledRotation) {isSetParameters=true, module=SetVariables_#7}
   /SetVariables(Particles.Position=$System.HeadSource) {isSetParameters=true, module=SetVariables_#8}
   &CameraOffset(Camera Offset Amount=100)
   &DynamicMaterialParameters(
    Index 0 Param 1=&MakeFloatFromVector(Vector=Linked() {linkedVariable {name=System.SampleSkeletalMesh.SampledPosition, type=NiagaraPosition}}, Channel=::NewEnumerator0)
    Index 0 Param 2=&MakeFloatFromVector(Vector=Linked() {linkedVariable {name=System.SampleSkeletalMesh.SampledPosition, type=NiagaraPosition}}, Channel=::NewEnumerator1)
    Index 0 Param 3=&MakeFloatFromVector(Vector=Linked() {linkedVariable {name=System.SampleSkeletalMesh.SampledPosition, type=NiagaraPosition}}, Channel=::NewEnumerator2)
    Index 0 Param 4=&Lerp_Float(
     A=-1.5
     B=1
     Alpha=&FloatFromCurve(FloatCurve=DataInterface() {propertyValues {Curve {keys [
          <at -0.0008 0.5892 RCIM_Cubic>
          {arriveTangent=0.4104, interpMode=RCIM_Cubic, leaveTangent=0.4104, time=0.035, value=0.0246}
          <at 1 1 RCIM_Cubic>
         ]}}})
    )
    Index 1 Param 1=&FloatFromCurve(
     FloatCurve=DataInterface() {propertyValues {Curve {keys [<at 0 1.75 RCIM_Cubic>, {arriveTangent=-1.75, interpMode=RCIM_Cubic, leaveTangent=-1.75, time=0.2, value=1.25}, <at 1 0 RCIM_Cubic>]}}}
     Scale Curve=1.25
    )
    Module.Param0WriteEnabled=true
    Module.Param0WriteEnabled1=true
    Module.Param1WriteEnabled=true
    Module.Param1WriteEnabled1=true
    Module.Param2WriteEnabled=true
    Module.Param2WriteEnabled1=true
    Module.Param3WriteEnabled=true
    Module.Param3WriteEnabled1=true
    Write Parameter Index 1=true
   )
   &SpriteFacingAndAlignment(Sprite Alignment=<0 0 1>, WriteAlignment=true, WriteFacing=true)
   &ScaleColor(
    Scale Alpha=&FloatFromCurve(FloatCurve=DataInterface() {propertyValues {Curve {keys [<at 0 1 RCIM_Cubic>,<at 1 0 RCIM_Cubic>]}}}, Scale Curve=1.25)
    ScaleA=true
    ScaleRGB=true
   ) {enabled=false}
  ]
  rendererClasses [&NiagaraSpriteRendererProperties]
  renderers [
   {rendererClass=&NiagaraSpriteRendererProperties}
   {rendererClass=&NiagaraSpriteRendererProperties, rendererIndex=1}
  ]
 }
 BodySmokeWash001 {
  emitterUpdateScript [
   &EmitterState(Life Cycle Mode=::NewEnumerator1, Loop Behavior=::NewEnumerator1, Loop Duration=1, UseLoopDelay=false)
   &SpawnBurst_Instantaneous(Use Spawn Probability=false, UseLoopCountLimit=false)
  ]
  particleSpawnScript [
   &InitializeParticle(
    Color=$System.PaleBlack
    Lifetime=2.75
    Position Offset=<0 0 86.4798>
    Sprite Size=<70 60>
    Sprite Size Mode=::NewEnumerator4
    UsePositionOffset=true
    Write Material Random=true
   )
   /SetVariables(Particles.Position=&AddVectorToPosition(Position=$System.HeadSource, Vector=<0 0 3>)) {isSetParameters=true}
  ]
  particleUpdateScript [
   &ParticleState()
   &SolveForcesAndVelocity(Clamp Velocity=false, Limit Acceleration=false)
   /SetVariables(Particles.Position=$System.SampleSkeletalMesh.SampledPosition) {enabled=false, isSetParameters=true, module=SetVariables_#6}
   /SetVariables(Particles.MeshOrientation=$System.SampleSkeletalMesh.SampledRotation) {isSetParameters=true, module=SetVariables_#7}
   /SetVariables(Particles.Position=/Niagara/DynamicInputs/Lerp/SmoothLerpOverTime_Position(
     Convergence Rate:NiagaraFloat=1
     Smooth Value:NiagaraPosition=$Particles.Position
     Target Value:NiagaraPosition=&AddVectorToPosition(Position=$System.HeadSource, Vector=<0 0 3>)
    )) {isSetParameters=true, module=SetVariables_#8}
   &CameraOffset(Camera Offset Amount=25) {enabled=false}
   &DynamicMaterialParameters(
    Index 0 Param 1=&MakeFloatFromVector(Vector=Linked() {linkedVariable {name=System.HeadSource, type=NiagaraPosition}}, Channel=::NewEnumerator0)
    Index 0 Param 2=&MakeFloatFromVector(Vector=Linked() {linkedVariable {name=System.HeadSource, type=NiagaraPosition}}, Channel=::NewEnumerator1)
    Index 0 Param 3=&MakeFloatFromVector(Vector=Linked() {linkedVariable {name=System.HeadSource, type=NiagaraPosition}}, Channel=::NewEnumerator2)
    Index 0 Param 4=&FloatFromCurve(FloatCurve=DataInterface() {propertyValues {Curve {keys [
         <at 0 0 RCIM_Cubic>
         {arriveTangent=1.9, interpMode=RCIM_Cubic, leaveTangent=1.9, time=0.1, value=1}
         {arriveTangent=-1.1111, interpMode=RCIM_Cubic, leaveTangent=-1.1111, time=0.5, value=0.95}
         <at 1 0 RCIM_Cubic>
        ]}}})
    Index 1 Param 1=&FloatFromCurve(FloatCurve=DataInterface() {propertyValues {Curve {keys [{interpMode=RCIM_Cubic, tangentMode=RCTM_User, time=0, value=1}, {interpMode=RCIM_Cubic, tangentMode=RCTM_User, time=1, value=0}]}}})
    Module.Param0WriteEnabled=true
    Module.Param0WriteEnabled1=true
    Module.Param1WriteEnabled=true
    Module.Param1WriteEnabled1=true
    Module.Param2WriteEnabled=true
    Module.Param2WriteEnabled1=true
    Module.Param3WriteEnabled=true
    Module.Param3WriteEnabled1=true
    Write Parameter Index 1=true
   )
   &SpriteFacingAndAlignment(Sprite Alignment=<0 0 1>, WriteAlignment=true, WriteFacing=true)
   &ScaleColor(
    Scale Alpha=&FloatFromCurve(FloatCurve=DataInterface() {propertyValues {Curve {keys [
         <at 0 0 RCIM_Cubic>
         {arriveTangent=0.6875, interpMode=RCIM_Cubic, leaveTangent=0.6875, time=0.1, value=1.025}
         {arriveTangent=-1.1389, interpMode=RCIM_Cubic, leaveTangent=-1.1389, time=0.4, value=0.275}
         <at 1 0 RCIM_Cubic>
        ]}}})
    ScaleA=true
    ScaleRGB=true
   )
   /SetVariables(Particles.FullSpriteSize:Vector2f=&Multiply_Vector2D_ByFloat(Float=0.55, Vector2D=$StackContext.SpriteSize)) {isSetParameters=true}
  ]
  rendererClasses [&NiagaraSpriteRendererProperties]
  renderers [
   {rendererClass=&NiagaraSpriteRendererProperties}
   {rendererClass=&NiagaraSpriteRendererProperties, rendererIndex=1}
   {rendererClass=&NiagaraSpriteRendererProperties, rendererIndex=2}
  ]
 }
 EyeEmbers {
  emitterUpdateScript [
   &EmitterState(Life Cycle Mode=::NewEnumerator1, Loop Behavior=::NewEnumerator1, Loop Delay=0.05, Loop Duration=1, UseLoopDelay=true)
   &SpawnBurst_Instantaneous(Spawn Count=10, Use Spawn Probability=false, UseLoopCountLimit=false)
   &SpawnRate(
    SpawnRate=&FloatFromCurve(
     FloatCurve=DataInterface() {propertyValues {Curve {keys [<at 0 1 RCIM_Cubic>, {arriveTangent=-3.3333, interpMode=RCIM_Cubic, leaveTangent=-3.3333, time=0.1, value=1.5}, <at 0.3 0 RCIM_Cubic>]}}}
     CurveIndex=$Emitter.NormalizedLoopAge
     Scale Curve=60
    )
    Use Spawn Probability=false
   )
  ]
  particleSpawnScript [
   &InitializeParticle(
    Color=<rgba 6 6 6 1>
    Lifetime=&RandomRangeFloat(Minimum=0.2, Maximum=2)
    Sprite Size=<1 1>
    Sprite Size Mode=::NewEnumerator4
    UsePositionOffset=false
    Write Material Random=true
   )
   &SkeletalMeshLocation(
    Bone Sampling Mode=::NewEnumerator4
    MaskSpawnGroup=false
    Sampled Position Offset=&Multiply_Vector(
     A=<4 10 2.75>
     B=&SelectVectorFromArray(
      Array Sampling Mode=::NewEnumerator0
      Direct Array Index=&SpawnGroup()
      Vector Selection Array=DataInterface() {propertyValues {FloatData [<1 1 1>, <1 1 -1>]}}
      Random Seed=0
     )
    )
    Skeletal Mesh=$User.SkeletalMesh
   )
   &AddVelocity(
    Cone Axis=&RotateVectorByQuaternion(Vector To Rotate=<0 1 0>, Quaternion=$StackContext.SkeletalMeshLocation.SampledRotation)
    Module.UseConeFalloff=false
    Module.UseVelDistribution=false
    Override Seed=false
    Rotation Coordinate Space=::World
    Velocity Mode=::NewEnumerator2
    Velocity Speed=&RandomRangeFloat(Minimum=5, Maximum=70)
   )
  ]
  particleUpdateScript [
   &ParticleState()
   &Drag(Drag=2, Use Linear Drag=true, Use Rotational Drag=true)
   &GravityForce(Gravity=&Multiply_VectorByFloat(Vector=<0 0 -400>, Float=&RandomRangeFloat(Minimum=-0.2, Maximum=0.1)))
   &CurlNoiseForce(Pan Noise=false)
   &SolveForcesAndVelocity(Clamp Velocity=false, Limit Acceleration=false)
   &ScaleColor(
    Linear Color Curve=DataInterface() {propertyValues {
      AlphaCurve {keys [<at 0 1>,<at 1 0>]}
      BlueCurve {keys [<at 0 0>,<at 0.1257 0>]}
      GreenCurve {keys [<at 0 1.2095>,<at 0.1257 0>]}
      RedCurve {keys [<at 0 2>,<at 0.1257 8>]}
     }}
    Scale Mode=::NewEnumerator2
   )
   &ScaleSpriteSize(
    Scale Factor=&MakeVector2D(
     X=1
     Y=&FloatFromCurve(FloatCurve=DataInterface() {propertyValues {Curve {keys [<at 0 0 RCIM_Cubic>,<at 0.1 1 RCIM_Cubic>,<at 1 0 RCIM_Cubic>]}}})
    )
    Scale Sprite Size Mode=::NewEnumerator2
   )
   &ScaleSpriteSizeBySpeed(Max Scale Factor=<1 10>, Velocity Threshold=200)
   &ScaleSpriteSize(
    Scale Sprite Size Mode=::NewEnumerator0
    Uniform Scale Factor=&MakeCustomFloatFromBool(
     Bool=&SetBoolByFloatComparison(
      A=&VectorLength(Vector=&SubtractVectorFromPosition(
        Position=&ReturnCameraPosition(Coordinate Space:ENiagaraCoordinateSpace=::World, Camera Query:NiagaraDataInterfaceCamera=DataInterface())
        Vector=Linked() {linkedVariable {name=Particles.Position, type=NiagaraPosition}}
       ))
      B=125
      Comparison Type=::NewEnumerator6
     )
     True Float=0
     False Float=1
    )
   ) {module=ScaleSpriteSize001}
  ]
  rendererClasses [&NiagaraSpriteRendererProperties]
  renderers [
   {rendererClass=&NiagaraSpriteRendererProperties}
   {rendererClass=&NiagaraSpriteRendererProperties, rendererIndex=1}
  ]
 }
 EyeGlows {
  emitterUpdateScript [
   &EmitterState(Life Cycle Mode=::NewEnumerator1, Loop Behavior=::NewEnumerator1, Loop Duration=1, UseLoopDelay=false)
   &SpawnBurst_Instantaneous(Use Spawn Probability=false, UseLoopCountLimit=false)
   &SpawnBurst_Instantaneous(Spawn Group=1, Use Spawn Probability=false, UseLoopCountLimit=false) {module=SpawnBurst_Instantaneous001}
  ]
  particleSpawnScript [&InitializeParticle(
    Color=<rgba 1 1 1 0.5>
    Lifetime=&RandomRangeFloat(Minimum=0.75, Maximum=1.2)
    Sprite Size=<15 5>
    Sprite Size Mode=::NewEnumerator4
    UsePositionOffset=false
    Write Material Random=true
   )]
  particleUpdateScript [
   &ParticleState()
   &ScaleSpriteSize(Uniform Curve Sprite Scale=DataInterface() {propertyValues {Curve {keys [<at 0 1 RCIM_Cubic>,<at 0.1 1.35 RCIM_Cubic>,<at 1 1 RCIM_Cubic>]}}})
   &SkeletalMeshLocation(
    Bone Sampling Mode=::NewEnumerator4
    MaskSpawnGroup=false
    Sampled Position Offset=&Multiply_Vector(
     A=<4 10 2.75>
     B=&SelectVectorFromArray(
      Direct Array Index=&SpawnGroup()
      Vector Selection Array=DataInterface() {propertyValues {FloatData [<1 1 1>, <1 1 -1>]}}
     )
    )
    Skeletal Mesh=$User.SkeletalMesh
   )
   &SolveForcesAndVelocity(Clamp Velocity=false, Limit Acceleration=false)
   &CameraOffset(Camera Offset Amount=&Multiply_Float(A=15, B=$System.ShowEyeGlow))
   &ScaleColor(
    Linear Color Curve=DataInterface() {propertyValues {
      bOptimizeLUT=false
      AlphaCurve {keys [<at 0 0.3>,<at 0.9607 0.7>]}
      BlueCurve {keys [<at 0 0>,<at 0.9636 0.0849>]}
      GreenCurve {keys [<at 0 0.6086>,<at 0.9636 1.0664>]}
      RedCurve {keys [<at 0 2>,<at 0.9636 2>]}
     }}
    Scale Mode=::NewEnumerator2
   )
   /SetVariables(Particles.LightPosition:NiagaraPosition=&AddVectorToPosition(Position=$Particles.Position, Vector=&TransformVector(Vector=<-2 -2 0>, Destination Space=::World))) {isSetParameters=true}
   &ScaleColor(
    Color Value To Scale=$StackContext.Color
    Scale Alpha=&FloatFromCurve(FloatCurve=DataInterface() {propertyValues {Curve {keys [<at 0.1 1.5 RCIM_Cubic>,<at 1 0 RCIM_Cubic>]}}}, CurveIndex=$System.Age)
    ScaleA=true
    ScaleRGB=true
   ) {module=ScaleColor001}
   /SetVariables(
    Particles.GlareColor:LinearColor=&MultiplyLinearColor(Linear Color=$Particles.Color, Scale Factor=<1 1 1 0.5>)
    Particles.Glaresize:Vector2f=&Multiply_Vector2D_ByFloat(
     Float=$System.ShowEyeGlow
     Vector2D=&Multiply_Vector2D(
      A=<20 37.3323>
      B=&MakeVector2D(
       X=&FloatFromCurve(FloatCurve=DataInterface() {propertyValues {Curve {keys [
            {arriveTangent=9.0725, interpMode=RCIM_Cubic, time=0, value=0.5}
            {arriveTangent=1.25, interpMode=RCIM_Cubic, leaveTangent=1.25, time=0.2, value=1.625}
            {interpMode=RCIM_Cubic, leaveTangent=-0.0405, time=1, value=1.75}
           ]}}})
       Y=1
      )
     )
    )
   ) {isSetParameters=true}
   &ScaleColor(Scale Alpha=$System.ShowEyeGlow, ScaleA=true, ScaleRGB=true) {module=ScaleColor002}
  ]
  rendererClasses [&NiagaraSpriteRendererProperties]
  renderers [
   {rendererClass=&NiagaraSpriteRendererProperties}
   {rendererClass=&NiagaraSpriteRendererProperties, rendererIndex=1}
  ]
 }
 FN_Debug {
  bEnabled=false
  emitterUpdateScript [
   &EmitterState()
   &SpawnBurst_Instantaneous(Spawn Count=1, Use Spawn Probability=false)
  ]
  particleSpawnScript [/Niagara/Modules/Spawn/Initialization/InitializeParticle(
    Lifetime=5
    Module.Write Color=true
    Module.Write Lifetime=true
    Module.Write Mass=true
    Module.Write Material Random=true
    Module.Write Position=true
    Module.Write Scale=false
    Module.Write SpriteRotation=false
    Module.Write SpriteSize=true
    Position=&AddVectorToPosition(Position=/Niagara/DynamicInputs/Helpers/SimulationPosition(), Vector=<0 0 150>)
    Sprite Size=<100 100>
   )]
  particleUpdateScript [
   &ParticleState()
   /SetVariables(Particles.Debug:NiagaraFloat=$System.ShowEyeGlow) {isSetParameters=true}
   &DynamicMaterialParameters(
    Index 0 Param 1=$Particles.Debug
    Module.Param0WriteEnabled=true
    Module.Param1WriteEnabled=true
    Module.Param2WriteEnabled=true
    Module.Param3WriteEnabled=true
   )
  ]
  rendererClasses [&NiagaraSpriteRendererProperties]
  renderers [{rendererClass=&NiagaraSpriteRendererProperties}]
 }
 FlameSplashSmoke {
  emitterUpdateScript [
   &EmitterState(Life Cycle Mode=::NewEnumerator1, Loop Behavior=::NewEnumerator1, Loop Duration=2, UseLoopDelay=false)
   &SpawnBurst_Instantaneous(Use Spawn Probability=false, UseLoopCountLimit=false)
  ]
  particleSpawnScript [
   /Niagara/Modules/Spawn/Initialization/V2/InitializeParticle(
    Color=<rgba 1 1 1 0.5>
    Lifetime=1.5
    Sprite Rotation Angle=270
    Sprite Rotation Mode=::NewEnumerator0
    Sprite Size=&Multiply_Vector2D_ByFloat(Float=0.35, Vector2D=<130 120>)
    Sprite Size Mode=::NewEnumerator4
    UsePositionOffset=false
    Write Material Random=true
   )
   &ShapeLocation(
    MaskSpawnGroup=false
    Offset=<0.7488 20 0>
    Offset Coordinate Space=::Local
    Offset Mode=::NewEnumerator0
    Override Seed=false
    Shape Origin=$System.HeadSource
    Sphere Radius=1
    UseOwnerScale=true
   )
  ]
  particleUpdateScript [
   &ParticleState()
   &ScaleSpriteSize(
    Uniform Curve Scale=&MakeCustomFloatFromBool(Bool=$User.bIsFrontEnd, True Float=0.55, False Float=1.2)
    Uniform Curve Sprite Scale=DataInterface() {propertyValues {Curve {keys [
        {arriveTangent=1.9622, interpMode=RCIM_Cubic, time=0, value=0.6}
        {arriveTangent=0.5896, interpMode=RCIM_Cubic, leaveTangent=0.5896, tangentMode=RCTM_User, time=0.3, value=0.8}
        {interpMode=RCIM_Cubic, leaveTangent=0.0224, time=1, value=1}
       ]}}}
   )
   &ScaleColor(
    Scale Alpha=&FloatFromCurve(FloatCurve=DataInterface() {propertyValues {Curve {keys [
         {interpMode=RCIM_Cubic, tangentMode=RCTM_User, time=0, value=0}
         {interpMode=RCIM_Cubic, tangentMode=RCTM_User, time=0.3, value=1}
         {tangentMode=RCTM_User, time=1, value=0}
        ]}}})
    Scale RGB=&MakeVectorFromLinearColorRGB(Color=&ColorFromCurve(ColorCurve=DataInterface() {propertyValues {
        AlphaCurve {keys [<at 0.4833 1 RCIM_Cubic>]}
        BlueCurve {keys [<at 0 0.17>,<at 1 0>]}
        GreenCurve {keys [<at 0 0.17>,<at 1 0>]}
        RedCurve {keys [<at 0 0.17>,<at 1 0>]}
       }}))
    ScaleA=true
    ScaleRGB=true
   )
   &ScaleColor(Scale RGB=&Multiply_VectorByFloat(Vector=<1 1 1>, Float=$System.ShowEyeGlow), ScaleA=true, ScaleRGB=true) {module=ScaleColor001}
   &DynamicMaterialParameters(
    Index 0 Param 1=&FloatFromCurve(
     FloatCurve=DataInterface() {propertyValues {Curve {keys [
         {arriveTangent=0.2738, interpMode=RCIM_Cubic, leaveTangent=0.2738, tangentMode=RCTM_User, time=0.1, value=0}
         {interpMode=RCIM_Cubic, tangentMode=RCTM_User, time=1, value=1}
        ]}}}
     Scale Curve=10
    )
    Index 0 Param 2=&FloatFromCurve(FloatCurve=DataInterface() {propertyValues {Curve {keys [
         {arriveTangent=0.0396, interpMode=RCIM_Cubic, leaveTangent=0.0396, tangentMode=RCTM_User, time=0, value=1.8}
         {arriveTangent=1.4204, interpMode=RCIM_Cubic, leaveTangent=1.4204, tangentMode=RCTM_User, time=1, value=2.5}
        ]}}})
    Index 0 Param 3=&FloatFromCurve(
     FloatCurve=DataInterface() {propertyValues {Curve {keys [
         {interpMode=RCIM_Cubic, tangentMode=RCTM_User, time=0, value=0}
         {arriveTangent=1.7963, interpMode=RCIM_Cubic, leaveTangent=1.7963, tangentMode=RCTM_User, time=1, value=1}
        ]}}}
     Scale Curve=0.1
    )
    Index 0 Param 4=&FloatFromCurve(
     FloatCurve=DataInterface() {propertyValues {Curve {keys [
         {arriveTangent=-0.768, interpMode=RCIM_Cubic, time=0, value=1}
         {arriveTangent=-0.875, interpMode=RCIM_Cubic, leaveTangent=-0.875, time=0.6, value=0.8}
         {interpMode=RCIM_Cubic, leaveTangent=-0.714, time=1, value=0.125}
        ]}}}
     Scale Curve=1.25
    )
    Module.Param0WriteEnabled=true
    Module.Param1WriteEnabled=true
    Module.Param2WriteEnabled=true
    Module.Param3WriteEnabled=true
   )
   &GravityForce(Gravity=<0 0 0>)
   /SetVariables(Particles.Position=&AddVectorToPosition(Position=$System.HeadSource, Vector=&TransformVector(Vector=<0 5 -5>, Destination Space=::Simulation))) {isSetParameters=true}
   &CameraOffset(Camera Offset Amount=20)
   &SolveForcesAndVelocity(Clamp Velocity=false, Limit Acceleration=false)
  ]
  rendererClasses [&NiagaraSpriteRendererProperties]
  renderers [{rendererClass=&NiagaraSpriteRendererProperties}]
 }
 HandsHide {
  emitterUpdateScript [
   &EmitterState(Life Cycle Mode=::NewEnumerator1, Loop Behavior=::NewEnumerator1, Loop Duration=1, UseLoopDelay=false)
   &SpawnBurst_Instantaneous(Spawn Group=1, Use Spawn Probability=false, UseLoopCountLimit=false)
   &SpawnBurst_Instantaneous(Spawn Group=2, Use Spawn Probability=false, UseLoopCountLimit=false) {module=SpawnBurst_Instantaneous001}
  ]
  particleSpawnScript [
   &InitializeParticle(
    Color=$System.PaleBlack
    Lifetime=&RandomRangeFloat(Minimum=1.1, Maximum=1.4)
    Sprite Size=<30 30>
    Sprite Size Mode=::NewEnumerator4
    UsePositionOffset=true
    Write Material Random=true
   )
   /SetVariables() {enabled=false, isSetParameters=true}
  ]
  particleUpdateScript [
   &ParticleState()
   &SolveForcesAndVelocity(Clamp Velocity=false, Limit Acceleration=false)
   &SkeletalMeshLocation(
    Bone / Socket Index=&SpawnGroup(Max Spawn Group=2)
    Bone Sampling Mode=::NewEnumerator4
    MaskSpawnGroup=false
    Sampled Position Offset=&SelectVectorFromArray(
     Direct Array Index=&SpawnGroup(Max Spawn Group=2)
     Vector Selection Array=DataInterface() {propertyValues {FloatData [<0 0 0>, <-9 0 0>, <9 0 0>]}}
    )
    Skeletal Mesh=$User.SkeletalMesh
   )
   /SetVariables(Particles.Position=$System.SampleSkeletalMesh.SampledPosition) {enabled=false, isSetParameters=true, module=SetVariables_#6}
   /SetVariables(Particles.MeshOrientation=$System.SampleSkeletalMesh.SampledRotation) {enabled=false, isSetParameters=true, module=SetVariables_#7}
   &CameraOffset(Camera Offset Amount=5)
   &DynamicMaterialParameters(
    Index 0 Param 1=&MakeFloatFromVector(Vector=Linked() {linkedVariable {name=Particles.Position, type=NiagaraPosition}}, Channel=::NewEnumerator0)
    Index 0 Param 2=&MakeFloatFromVector(Vector=Linked() {linkedVariable {name=Particles.Position, type=NiagaraPosition}}, Channel=::NewEnumerator1)
    Index 0 Param 3=&MakeFloatFromVector(Vector=Linked() {linkedVariable {name=Particles.Position, type=NiagaraPosition}}, Channel=::NewEnumerator2)
    Index 0 Param 4=&FloatFromCurve(FloatCurve=DataInterface() {propertyValues {Curve {keys [
         <at 0 0 RCIM_Cubic>
         {arriveTangent=1.9, interpMode=RCIM_Cubic, leaveTangent=1.9, time=0.1, value=1}
         {arriveTangent=-1.1111, interpMode=RCIM_Cubic, leaveTangent=-1.1111, time=0.5, value=0.95}
         <at 1 0 RCIM_Cubic>
        ]}}})
    Index 1 Param 1=&FloatFromCurve(FloatCurve=DataInterface() {propertyValues {Curve {keys [{interpMode=RCIM_Cubic, tangentMode=RCTM_User, time=0, value=1}, {interpMode=RCIM_Cubic, tangentMode=RCTM_User, time=1, value=0}]}}})
    Module.Param0WriteEnabled=true
    Module.Param0WriteEnabled1=true
    Module.Param1WriteEnabled=true
    Module.Param1WriteEnabled1=true
    Module.Param2WriteEnabled=true
    Module.Param2WriteEnabled1=true
    Module.Param3WriteEnabled=true
    Module.Param3WriteEnabled1=true
    Write Parameter Index 1=true
   )
   &ScaleColor(
    Scale Alpha=&FloatFromCurve(FloatCurve=DataInterface() {propertyValues {Curve {keys [<at 0 0 RCIM_Cubic>,<at 0.3 1 RCIM_Cubic>,<at 1 0 RCIM_Cubic>]}}}, Scale Curve=1.5)
    ScaleA=true
    ScaleRGB=true
   )
  ]
  rendererClasses [&NiagaraSpriteRendererProperties]
  renderers [
   {rendererClass=&NiagaraSpriteRendererProperties}
   {rendererClass=&NiagaraSpriteRendererProperties, rendererIndex=1}
  ]
 }
 SimpleSpriteBurst {
  bEnabled=false
  emitterUpdateScript [
   &EmitterState(Life Cycle Mode=::NewEnumerator1, Loop Behavior=::NewEnumerator1, Loop Delay=$System.TransitionTime, Loop Duration=2, UseLoopDelay=true)
   &SpawnBurst_Instantaneous(Spawn Count=10, Use Spawn Probability=false, UseLoopCountLimit=false)
  ]
  particleSpawnScript [
   &InitializeParticle(Lifetime=2, Sprite Size Mode=::NewEnumerator3, UsePositionOffset=false, Write Material Random=true)
   &AddVelocity(Module.Use Falloff Distance=false, Override Seed=false, Velocity Mode=::NewEnumerator1)
   &SkeletalMeshLocation(Bone Sampling Mode=::NewEnumerator4, MaskSpawnGroup=false, Skeletal Mesh=$User.SkeletalMesh)
  ]
  particleUpdateScript [
   &ParticleState()
   &ScaleColor(
    Scale Alpha=&FloatFromCurve(FloatCurve=DataInterface() {propertyValues {Curve {keys [<at -0.0046 1.0113 RCIM_Cubic>,<at 1.0011 0.002 RCIM_Cubic>]}}})
    ScaleA=true
    ScaleRGB=false
   )
   &SolveForcesAndVelocity(Clamp Velocity=false, Limit Acceleration=false)
  ]
  rendererClasses [&NiagaraSpriteRendererProperties]
  renderers [{rendererClass=&NiagaraSpriteRendererProperties}]
 }
 SmokeFaceWash {
  emitterUpdateScript [
   &EmitterState(Life Cycle Mode=::NewEnumerator1, Loop Behavior=::NewEnumerator1, Loop Duration=1, UseLoopDelay=false)
   &SpawnBurst_Instantaneous(Use Spawn Probability=false, UseLoopCountLimit=false)
  ]
  particleSpawnScript [&InitializeParticle(Color=<rgba 0.005 0.005 0.005 5>, Lifetime=1.2, Sprite Size Mode=::NewEnumerator3, UsePositionOffset=false, Write Material Random=true)]
  particleUpdateScript [
   &ParticleState()
   &SolveForcesAndVelocity(Clamp Velocity=false, Limit Acceleration=false)
   /SetVariables(Particles.Position=$System.SampleSkeletalMesh.SampledPosition) {isSetParameters=true, module=SetVariables_#6}
   /SetVariables(Particles.MeshOrientation=$System.SampleSkeletalMesh.SampledRotation) {isSetParameters=true, module=SetVariables_#7}
   &CameraOffset(Camera Offset Amount=15)
   &DynamicMaterialParameters(
    Index 0 Param 1=&Lerp_Float(
     A=1
     B=0
     Alpha=&FloatFromCurve(FloatCurve=DataInterface() {propertyValues {Curve {keys [<at 0 1 RCIM_Cubic>, {arriveTangent=-1, interpMode=RCIM_Cubic, leaveTangent=-1, time=0.5, value=0.575}, <at 1 0 RCIM_Cubic>]}}})
    )
    Module.Param0WriteEnabled=true
    Module.Param1WriteEnabled=true
    Module.Param2WriteEnabled=true
    Module.Param3WriteEnabled=true
   )
   &ScaleColor(
    Scale Alpha=&FloatFromCurve(FloatCurve=DataInterface() {propertyValues {Curve {keys [
         {interpMode=RCIM_Cubic, tangentMode=RCTM_User, time=0, value=0}
         {interpMode=RCIM_Cubic, tangentMode=RCTM_User, time=0.2, value=1}
         {tangentMode=RCTM_User, time=1, value=0}
        ]}}})
    ScaleA=true
    ScaleRGB=true
   )
  ]
  rendererClasses [&NiagaraMeshRendererProperties]
  renderers [{rendererClass=&NiagaraMeshRendererProperties, rendererIndex=0}]
 }
}
summary {
 systemName=NS_HeavyRoarEmote_Style_Transform
 emitters [
  {bEnabled=false, emitterName=SimpleSpriteBurst, rendererClasses [&NiagaraSpriteRendererProperties]}
  {emitterName=FlameSplashSmoke, rendererClasses [&NiagaraSpriteRendererProperties]}
  {emitterName=SmokeFaceWash, rendererClasses [&NiagaraMeshRendererProperties]}
  {bEnabled=false, emitterName=BodySmokeWash, rendererClasses [&NiagaraSpriteRendererProperties]}
  {emitterName=EyeGlows, rendererClasses [&NiagaraSpriteRendererProperties]}
  {emitterName=EyeEmbers, rendererClasses [&NiagaraSpriteRendererProperties]}
  {emitterName=BodySmokeWash001, rendererClasses [&NiagaraSpriteRendererProperties]}
  {emitterName=HandsHide, rendererClasses [&NiagaraSpriteRendererProperties]}
  {bEnabled=false, emitterName=FN_Debug, rendererClasses [&NiagaraSpriteRendererProperties]}
 ]
 userVariables [
  User.bIsFrontEnd {type=NiagaraBool}
  User.ImpactColor {type=LinearColor, description="", defaultValue {struct=LinearColor, value=<rgba 1 1 1 1>}}
  User.IsWeakPoint {type=NiagaraBool}
  User.SkeletalMesh {
   type=NiagaraDataInterfaceSkeletalMesh
   defaultValue {
    struct=NiagaraExt_VariableValue_DataInterface
    value {dataInterface=&self:NiagaraDataInterfaceSkeletalMesh_0}
   }
  }
 ]
}
